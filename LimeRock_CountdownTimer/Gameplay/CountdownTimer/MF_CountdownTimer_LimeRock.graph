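# Material function that draws the countdown timer readout from a number strip texture, outputting greyscale, chromatic-aberration color and mask.
# Center-scaled UVs split into 11 default slots, where nested Ifs pick MPC_Countdown Digit8 through Digit1, with -1 colon separators between pairs.
# The floored value plus the slot's fraction indexes the strip.
# Samples use derivatives based on the unremapped UVs, and two horizontally offset taps supply red and blue fringes.
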MaterialFunction /LimeRock_CountdownTimer/Gameplay/CountdownTimer/MF_CountdownTimer_LimeRock
graph {
 properties {description="outputs scalar value as number from texture -1==: /// -2==space"}
 nodes [
  n1 = FunctionInput(Preview=TextureCoordinate() {coordinateIndex=2}) {bUsePreviewValueAsDefault=true, inputName=UVs, inputType=FunctionInput_Vector2, sortPriority=1}
  n2 = ScalarParameter() {defaultValue=0.9128, group=Numbers, parameterName=u_NumberPaddingScale}
  n3 = ScalarParameter() {defaultValue=0.3143, group=Numbers, parameterName=v_NumberPaddingScale}
  n4 = MaterialFunctionCall(
   UVs=@n1
   Texture Scale=FunctionInput(Preview=AppendVector(A=@n2, B=@n3)) {bUsePreviewValueAsDefault=true, inputName=UVPadding, inputType=FunctionInput_Vector2, sortPriority=1}
  ) {materialFunction=&ScaleUVsByCenter}
  n5 = FunctionInput(Preview=Constant() {desc=Number of digits, r=11}) {
   bUsePreviewValueAsDefault=true
   desc=Only whole numbers between 1 and 8 Using a constant saves some instructions
   description=Using a constant saves some instructions
   inputName="NumberofDigits(int)"
   inputType=FunctionInput_Scalar
   sortPriority=2
  }
  n6 = Multiply(A=ComponentMask(None=@n4.UVs) {r=true}, B=@n5) {constB=5}
  n7 = Constant() {r=2}
  n8 = Constant() {r=5}
  n9 = Constant() {r=8}
  n10 = Constant() {r=11}
  n11 = Subtract(A=Constant() {r=11}, B=Subtract(A=@n5, B=@n6))
  n12 = ComponentMask(None=@n11) {r=true}
  n13 = CollectionParameter() {collection=/Game/Athena/Events/Generic/CountdownTimer/Materials/MPC_Countdown, parameterName=Digit8}
  n14 = If(
   A=Constant() {r=1}
   B=@n12
   "A > B"=@n13
   "A < B"=CollectionParameter() {collection=&MPC_Countdown, parameterName=Digit7}
  )
  n15 = If(A=Constant() {r=2}, B=@n12, "A > B"=@n14, "A < B"=Constant() {r=-1})
  n16 = If(
   A=Constant() {r=3}
   B=@n12
   "A > B"=@n15
   "A < B"=CollectionParameter() {collection=&MPC_Countdown, parameterName=Digit6}
  )
  n17 = If(
   A=Constant() {r=4}
   B=@n12
   "A > B"=@n16
   "A < B"=CollectionParameter() {collection=&MPC_Countdown, parameterName=Digit5}
  )
  n18 = If(A=Constant() {r=5}, B=@n12, "A > B"=@n17, "A < B"=Constant() {r=-1})
  n19 = If(
   A=Constant() {r=6}
   B=@n12
   "A > B"=@n18
   "A < B"=CollectionParameter() {collection=&MPC_Countdown, parameterName=Digit4}
  )
  n20 = If(
   A=Constant() {r=7}
   B=@n12
   "A > B"=@n19
   "A < B"=CollectionParameter() {collection=&MPC_Countdown, parameterName=Digit3}
  )
  n21 = If(A=Constant() {r=8}, B=@n12, "A > B"=@n20, "A < B"=Constant() {r=-1})
  n22 = If(
   A=Constant() {r=9}
   B=@n12
   "A > B"=@n21
   "A < B"=CollectionParameter() {collection=&MPC_Countdown, parameterName=Digit2}
  )
  n23 = If(
   A=Constant() {r=10}
   B=@n12
   "A > B"=@n22
   "A < B"=CollectionParameter() {collection=&MPC_Countdown, parameterName=Digit1}
  )
  n24 = If(
   A=@n11
   B=Constant() {r=9}
   "A > B"=If(A=@n11, B=Constant() {r=10}, "A > B"=If(A=@n11, B=@n10, "A > B"=@n23, "A == B"=@n23, "A < B"=@n23), "A == B"=@n22, "A < B"=@n22)
   "A == B"=@n21
   "A < B"=@n21
  )
  n25 = If(
   A=@n11
   B=Constant() {r=6}
   "A > B"=If(A=@n11, B=Constant() {r=7}, "A > B"=If(A=@n11, B=@n9, "A > B"=@n24, "A == B"=@n20, "A < B"=@n20), "A == B"=@n19, "A < B"=@n19)
   "A == B"=@n18
   "A < B"=@n18
  )
  n26 = If(
   A=@n11
   B=Constant() {r=3}
   "A > B"=If(A=@n11, B=Constant() {r=4}, "A > B"=If(A=@n11, B=@n8, "A > B"=@n25, "A == B"=@n17, "A < B"=@n17), "A == B"=@n16, "A < B"=@n16)
   "A == B"=@n15
   "A < B"=@n15
  )
  n27 = Floor(None=If(A=@n11, B=Constant() {r=1}, "A > B"=If(A=@n11, B=@n7, "A > B"=@n26, "A == B"=@n14, "A < B"=@n14), "A == B"=@n13, "A < B"=@n13))
  n28 = Divide(A=Add(A=Frac(None=@n6), B=@n27), B=ScalarParameter() {defaultValue=11, group=Numbers, parameterName=Numbers of Digits On Texture}) {constB=11}
  n29 = AppendVector(A=@n28, B=ComponentMask(None=@n4.UVs) {g=true})
  n30 = DDX(Value=@n1)
  n31 = DDY(Value=@n1)
  n32 = TextureSampleParameter2D!(UVs=@n29, "DDX(UVs)"=@n30, "DDY(UVs)"=@n31, Apply View MipBias) {
   group=Numbers
   mipValueMode=TMVM_Derivative
   parameterName=NumbersTexture
   samplerType=SAMPLERTYPE_Masks
   texture=/Game/Athena/Playlists/Barrier/Textures/T_Numbers2_M
  }
  n33 = Multiply(A=@n32.R, B=@n4.0-1 mask)
  FunctionOutput(None=@n33) {outputName=NumbersGreyscale}
  FunctionOutput(None=@n4.0-1 mask) {outputName=0-1 Mask, sortPriority=2}
  n34 = FunctionInput(Preview=Constant() {r=0.005}) {bUsePreviewValueAsDefault=true, inputName=Abberation Distance, inputType=FunctionInput_Scalar, sortPriority=7}
  n35 = TextureSampleParameter2D(
   UVs=Add(A=@n29, B=MaterialFunctionCall(X=@n34) {materialFunction=&MakeFloat2}.Result)
   "DDX(UVs)"=Add(A=@n34, B=@n30)
   "DDY(UVs)"=@n31
  ) {
   group=Numbers
   mipValueMode=TMVM_Derivative
   parameterName=NumbersTexture
   samplerType=SAMPLERTYPE_Masks
   texture=&T_Numbers2_M
  }
  n36 = MaterialFunctionCall(X=Multiply(A=@n34) {constB=-1}) {materialFunction=&MakeFloat2}
  n37 = TextureSampleParameter2D(UVs=Add(A=@n29, B=@n36.Result), "DDX(UVs)"=Add(A=@n30, B=@n36.Result), "DDY(UVs)"=@n31) {
   group=Numbers
   mipValueMode=TMVM_Derivative
   parameterName=NumbersTexture
   samplerType=SAMPLERTYPE_Masks
   texture=&T_Numbers2_M
  }
  FunctionOutput(None=MaterialFunctionCall(X=Multiply(A=@n35.R, B=@n4.0-1 mask), Y=@n33, Z=Multiply(A=@n4.0-1 mask, B=@n37.R)) {materialFunction=&MakeFloat3}.Result) {outputName=NumbersColor, sortPriority=1}
  FunctionOutput(None=@n4.UVs) {outputName=UVS, sortPriority=8}
  FunctionOutput(None=@n4.0-1 mask) {outputName=Mask, sortPriority=9}
  If(A=@n11, B=Constant() {r=1}, "A > B"=@n23, "A == B"=@n23, "A < B"=@n23)
  OneMinus(None=@n27)
  OneMinus(None=@n28)
  ScalarParameter() {defaultValue=5, group=Numbers, parameterName=Digit_5}
  ScalarParameter() {defaultValue=4, group=Numbers, parameterName=Digit_4}
  ScalarParameter() {defaultValue=2, group=Numbers, parameterName=Digit_2}
  ScalarParameter() {defaultValue=1, group=Numbers, parameterName=Digit_1}
  ScalarParameter() {defaultValue=-1, group=Numbers, parameterName=Digit_3}
  ScalarParameter() {defaultValue=-1, group=Numbers, parameterName=Digit_6}
  ScalarParameter() {defaultValue=7, group=Numbers, parameterName=Digit_7}
  ScalarParameter() {defaultValue=6.5222, group=Numbers, parameterName=Digit_8}
 ]
}
