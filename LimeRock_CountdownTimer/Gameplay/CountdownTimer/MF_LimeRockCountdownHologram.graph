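# Material function for the LimeRock countdown timer hologram, outputting material attributes, mask, color, WPO and random flicker.
# Digits from MF_CountdownTimer_LimeRock and DHMS labels blend with a background image and radial number cutout, driven by MPC_CountdownTimerVisuals.
# Glitch-warped UVs, chromatic aberration, flickering noise and distance-adjusted scanlines form the holographic look.
# WPO yaws it toward the camera, scales it with distance and displaces it far below when MPC_Countdown hides it.
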MaterialFunction /LimeRock_CountdownTimer/Gameplay/CountdownTimer/MF_LimeRockCountdownHologram
graph {nodes [
  n1 = NamedRerouteUsage()
  n2 = NamedRerouteUsage()
  n3 = LinearInterpolate(
   A=@n2
   B=LinearInterpolate(A=NamedRerouteUsage(), B=@n2, Alpha=NamedRerouteUsage())
   Alpha=CollectionParameter() {collection=&/MPC_CountdownTimerVisuals, parameterName=Timer_UseRandomFlicker}
  )
  n4 = Saturate(None=Add(A=@n3, B=@n1))
  n5 = Add(A=@n3, B=LinearInterpolate(A=@n1, Alpha=NamedRerouteUsage()) {constB=0})
  n6 = CollectionParameter() {collection=&MPC_CountdownTimerVisuals, parameterName=Timer_ImageCutoutNumbers}
  n7 = MaterialFunctionCall() {materialFunction=&LinearGradient}
  n8 = OneMinus(None=@n7.UGradient)
  n9 = OneMinus(None=@n7.VGradient)
  n10 = ScalarParameter() {group=Glitch, parameterName=Image.UseNumbersCutout}
  n11 = NamedRerouteUsage()
  n12 = LinearInterpolate(
   B=LinearInterpolate(
    A=@n10
    B=LinearInterpolate(B=@n10, Alpha=@n11)
    Alpha=CollectionParameter() {collection=&MPC_CountdownTimerVisuals, parameterName=Timer_UseRandomFlicker}
   )
   Alpha=CollectionParameter() {collection=&MPC_CountdownTimerVisuals, parameterName=Timer_ShowImage}
  )
  n13 = CollectionParameter() {collection=&MPC_CountdownTimerVisuals, parameterName=Timer_ImageCutoutNumbers}
  n14 = Multiply(
   A=MaterialFunctionCall(X=ConstantBiasScale(None=@n8) {bias=-0.5, scale=2}, Y=ConstantBiasScale(None=@n9) {bias=-0.5, scale=2}) {materialFunction=&MakeFloat2}.Result
   B=Multiply(A=LinearInterpolate(B=@n12, Alpha=@n13), B=ScalarParameter() {group=Glitch, parameterName=Image.NumbersCutout.Warp})
  )
  n15 = MaterialFunctionCall() {materialFunction=&RadialGradientExponential}
  n16 = TextureCoordinate()
  n17 = MaterialFunctionCall(Float2=TextureCoordinate()) {materialFunction=&BreakOutFloat2Components}
  n18 = Multiply(A=@n17.R) {constB=0.75}
  n19 = Time()
  n20 = ScalarParameter() {defaultValue=0.2, group=Glitch, parameterName=Glitch.PanSpeed}
  n21 = TextureSample(UVs=Add(
    A=MaterialFunctionCall(X=@n18, Y=@n17.G) {materialFunction=&MakeFloat2}.Result
    B=Multiply(A=@n19, B=@n20) {constB=0.3}
   )) {
   samplerType=SAMPLERTYPE_Masks
   texture=/Game/Environments/Asteria/Sets/NeonCity/Techart/Holograms/F_MED_BlueBadass_NeonLines_-_GlitchNoise_v1
  }
  n22 = LinearInterpolate(A=@n21.R, B=Round(None=@n21.R), Alpha=ScalarParameter() {group=Glitch, parameterName=Glitch.Hardness})
  n23 = ScalarParameter() {defaultValue=0.02, group=Glitch, parameterName=Glitch.Warp}
  n24 = Add(A=TextureCoordinate()) {constB=0.33}
  n25 = Multiply(A=Frac(None=Time())) {constB=0.02}
  n26 = TextureCoordinate()
  n27 = Constant() {r=5}
  n28 = Multiply(A=Frac(None=Time()), B=@n27) {constB=12}
  n29 = Multiply(A=Divide(A=Floor(None=@n28), B=Subtract(A=@n27)) {constB=11}, B=Constant2Vector() {g=1, r=-1})
  n30 = ScalarParameter() {defaultValue=0.2, group=Glitch, parameterName=Glitch.PanSpeed}
  n31 = TextureSample(UVs=Add(A=Multiply(A=@n26) {constB=0.5}, B=Multiply(A=@n29, B=@n30) {constB=0.3})) {samplerType=SAMPLERTYPE_LinearColor, texture=/Game/Characters/Player/Common/MaterialFunctions/Superlevel_DigitalCamo/SuperCamo_M}
  n32 = Time()
  n33 = SmoothStep(Value=Sine(None=Divide(A=@n32) {constB=1.6765})) {constMin=-1}
  n34 = Frac(None=Add(A=@n32, B=Multiply(B=@n33) {constA=0.1}))
  n35 = Divide(A=Floor(None=Multiply(A=@n34) {constB=3}))
  n36 = MaterialFunctionCall(A=@n31.R, B=@n31.R, C=@n31.R, Alpha=@n35) {materialFunction=&Lerp_3Color}
  n37 = TextureSample(UVs=Add(A=Add(A=@n24, B=@n25), B=Multiply(A=@n36.Result) {constB=0.1})) {
   samplerType=SAMPLERTYPE_Masks
   texture=&F_MED_BlueBadass_NeonLines_-_GlitchNoise_v1
  }
  n38 = Round(None=Power(Base=@n37.R) {constExponent=4})
  n39 = Constant() {r=0.5}
  n40 = MaterialFunctionCall(A=@n31.B, B=@n31.B, C=@n31.B, Alpha=@n35) {materialFunction=&Lerp_3Color}
  n41 = TextureSample(UVs=Add(A=TextureCoordinate(), B=Multiply(A=@n40.Result) {constB=0.1})) {
   samplerType=SAMPLERTYPE_Masks
   texture=&F_MED_BlueBadass_NeonLines_-_GlitchNoise_v1
  }
  n42 = Round(None=Power(Base=@n41.R) {constExponent=4})
  n43 = Subtract(A=Min(
    A=MaterialFunctionCall(Input=@n38, Target Low=@n39) {materialFunction=&RemapValueRange}.Result
    B=OneMinus(None=@n42)
   )) {constB=0.5}
  n44 = Add(
   A=Add(A=Multiply(A=@n22, B=@n23) {constB=0.02}, B=Multiply(A=@n43) {constB=2})
   B=ScalarParameter() {group=Glitch, parameterName=Glitch.UVOffset}
  ) {constB=-0.5, desc=Recenter}
  n45 = ScalarParameter() {defaultValue=0.02, group=Glitch, parameterName=Glitch.Intensity}
  n46 = Add(A=@n16, B=NamedRerouteDeclaration(None=Multiply(A=@n44, B=@n45) {constB=0.02}) {name=UVGlitching, nodeColor=<rgba 0 0.7422 1 1>})
  n47 = CollectionParameter() {collection=&MPC_CountdownTimerVisuals, parameterName=Timer_NumbersGlitchAmount}
  n48 = Add(
   A=Add(A=LinearInterpolate(B=@n14, Alpha=@n15.RadialGradientExponential), B=LinearInterpolate(A=@n16, B=@n46, Alpha=@n47))
   B=AppendVector(
    A=ScalarParameter() {group=Numbers, parameterName=u_NumberPaddingOffset}
    B=ScalarParameter() {group=Numbers, parameterName=v_NumberPaddingOffset}
   )
  )
  n49 = Add(A=AppendVector(A=ScalarParameter() {parameterName=DHMS.Pos.X}, B=ScalarParameter() {parameterName=DHMS.Pos.Y}), B=@n48)
  n50 = AppendVector(A=ScalarParameter() {defaultValue=1, parameterName=DHMS.Scale.X}, B=ScalarParameter() {defaultValue=1, parameterName=DHMS.Scale.Y})
  n51 = Multiply(
   A=TextureSampleParameter2D(UVs=MaterialFunctionCall(UVs=@n49, Texture Scale=@n50) {materialFunction=&ScaleUVsByCenter}.UVs) {parameterName=DHMS.Image, samplerType=SAMPLERTYPE_Masks, texture=/Game/Athena/Playlists/Barrier/Textures/Countdown_DHMS}.RGB
   B=ScalarParameter() {defaultValue=1, parameterName=DHMS.Brightness}
  )
  n52 = MaterialFunctionCall!(
   UVs=@n48
   UVPadding=AppendVector(
    A=ScalarParameter() {group=Numbers, parameterName=u_NumberPaddingScale}
    B=ScalarParameter() {group=Numbers, parameterName=v_NumberPaddingScale}
   )
   "NumberofDigits(int)"=ScalarParameter() {defaultValue=11, parameterName=NumberofDigits}
   Abberation Distance=Multiply(
    A=ScalarParameter() {defaultValue=0.03, group=Glitch, parameterName=Numbers.AbberationDistance}
    B=CollectionParameter() {collection=&MPC_CountdownTimerVisuals, parameterName=Timer_ChromAbberationPower}
   )
  ) {materialFunction=&/MF_CountdownTimer_LimeRock}
  n53 = Saturate(None=LinearInterpolate(
    B=Add(A=@n51, B=@n52.NumbersColor)
    Alpha=CollectionParameter() {collection=&MPC_CountdownTimerVisuals, parameterName=Timer_ShowNumbers}
   ))
  n54 = Saturate(None=LinearInterpolate(
    A=NamedRerouteDeclaration(None=@n53) {name=Numbers.FinalMask, nodeColor=<rgba 0 0.6953 1 1>}
    B=LinearInterpolate(A=@n4, B=@n5, Alpha=@n6)
    Alpha=CollectionParameter() {collection=&MPC_CountdownTimerVisuals, parameterName=Timer_ShowImage}
   ))
  n55 = Multiply(
   A=SphereMask(A=TextureCoordinate(), B=Constant() {r=0.5}, Radius=ScalarParameter() {defaultValue=0.5, parameterName=Background.Radius}) {hardnessPercent=0}
   B=ScalarParameter() {defaultValue=0.5, parameterName=BackgroundMult}
  ) {constB=0.15}
  n56 = ScalarParameter() {defaultValue=1500, group=Noise, parameterName=NoiseSize}
  n57 = Multiply(A=Time(), B=VectorParameter() {group=Noise, parameterName=NoisePanSpeed, defaultValue=<rgba -0.1 0.09 -0.15 0>}.RGB)
  n58 = MaterialFunctionCall(
   TextureObject=TextureObject() {samplerType=SAMPLERTYPE_LinearColor, texture=&T_NoisePack_LinearColor}
   TextureSize=Multiply(A=VectorParameter() {group=Noise, parameterName=NoiseTileMult, defaultValue=<rgba 2 2 0.1 1>}.RGB, B=@n56)
   WorldPosition=Add(A=WorldPosition().XYZ, B=Multiply(A=@n56, B=@n57))
  ) {materialFunction=&WorldAlignedTexture}
  n59 = Time()
  n60 = Multiply(A=ScalarParameter() {defaultValue=1, parameterName=FlickerSpeed}, B=@n59)
  n61 = AppendVector(A=Sine(None=@n60) {period=1.6}, B=Sine(None=@n60) {period=1.42})
  n62 = Sine(None=@n60) {period=1.51}
  n63 = Sine(None=@n60) {period=2.1}
  n64 = ScalarParameter() {defaultValue=0.5, group=Noise, parameterName=NoiseFrequency}
  n65 = Multiply(B=@n64) {constA=0.25}
  n66 = Add(A=Sine(None=@n60) {period=0.72}, B=Multiply(B=@n64) {constA=0.75}) {constB=0.75}
  n67 = Add(A=Multiply(B=@n64) {constA=0.9, constB=0.9}, B=Sine(None=@n60) {period=1.25}) {constB=0.9}
  n68 = Multiply(A=Multiply(A=Add(A=@n63, B=@n65), B=@n66), B=Multiply(A=@n66, B=Ceil(None=@n67)))
  n69 = ConstantBiasScale(None=Sine(None=@n59) {period=0.075})
  n70 = Multiply(A=Ceil(None=AppendVector(A=@n61, B=@n62)), B=Multiply(A=Saturate(None=@n68), B=LinearInterpolate(Alpha=@n69) {constA=0.7}))
  n71 = ScalarParameter() {defaultValue=2, group=Noise, parameterName=NoiseMult}
  n72 = Multiply(
   A=Multiply(A=Power(Base=@n58.XY Texture) {constExponent=4}, B=ScalarParameter() {defaultValue=2, group=Noise, parameterName=NoiseMaskMult}) {constB=20}
   B=Add(A=ScalarParameter() {defaultValue=0.15, group=Noise, parameterName=NoiseAdd}, B=Multiply(A=@n70, B=@n71))
  )
  n73 = MaterialFunctionCall(In=Desaturation(None=Saturate(None=@n72)), Contrast=ScalarParameter() {defaultValue=0.2, group=Noise, parameterName=NoiseContrast}) {materialFunction=&CheapContrast}
  n74 = Multiply(A=@n73.Result) {constB=2}
  n75 = Subtract(A=ScalarParameter() {defaultValue=0.3, parameterName=MinScanlineOpacity}, B=@n74)
  n76 = Saturate(None=@n73.Result)
  n77 = Multiply(A=ScalarParameter() {defaultValue=200, parameterName=ScanlineTile}, B=LinearInterpolate(Alpha=Ceil(None=@n76)) {constA=1, constB=1.25})
  n78 = ScalarParameter() {defaultValue=30000, parameterName=ScaleDistance}
  n79 = Divide(A=Distance(A=CameraPositionWS(), B=ActorPositionWS()), B=@n78) {constB=1000}
  n80 = Divide(A=ComponentMask(None=WorldPosition().XYZ) {b=true}, B=LinearInterpolate(B=@n78, Alpha=Saturate(None=@n79)) {constA=50, constB=100}) {constB=20}
  n81 = ConstantBiasScale(None=Sine(None=Multiply(A=@n77, B=@n80) {constA=1}))
  n82 = Saturate(None=LinearInterpolate(A=Clamp(None=@n75) {minDefault=-0.25}, B=Add(A=@n74), Alpha=Power(Base=@n81) {constExponent=1.5}) {constA=0.5})
  n83 = ScalarParameter() {defaultValue=1.25, parameterName=OpacityMult}
  n84 = NamedRerouteDeclaration(None=Multiply(A=@n55, B=Multiply(A=@n82, B=@n83))) {name=Scanlines.Mask, nodeColor=<rgba 1 0 0.8437 1>}
  n85 = Add(A=NamedRerouteDeclaration(None=@n55) {name=BackgroundShape, nodeColor=<rgba 0.6406 1 0 1>}, B=@n84)
  n86 = ScalarParameter() {parameterName=BackgroundFill}
  n87 = Saturate(None=Add(A=@n54, B=LinearInterpolate(A=@n84, B=@n85, Alpha=@n86)))
  n88 = ObjectPositionWS()
  n89 = WorldPosition()
  n90 = CollectionParameter() {collection=/Game/Athena/Events/Generic/CountdownTimer/Materials/MPC_Countdown, parameterName=ShowCountdown}
  n91 = Multiply(
   A=@n90
   B=CollectionParameter() {collection=&MPC_Countdown, parameterName=ShowCountdownPhase2}
  )
  n92 = Constant3Vector() {constant=<rgba 0 0 1 0>}
  n93 = CameraPositionWS()
  n94 = ActorPositionWS()
  n95 = Normalize(VectorInput=Multiply(A=Constant3Vector() {constant=<rgba 1 1 0 0>}, B=Subtract(A=@n93, B=@n94)))
  n96 = Normalize(VectorInput=Constant3Vector() {constant=<rgba 0 1 0 0>})
  n97 = DotProduct(A=@n95, B=Normalize(VectorInput=Constant3Vector() {constant=<rgba 1 0 0 0>}))
  n98 = Multiply(
   A=Divide(A=MaterialFunctionCall!(Normalized V1=@n95, Normalized V2=@n96) {materialFunction=/Engine/Functions/Engine_MaterialFunctions02/Math/AngleBetweenVectors}.Angle In Degrees) {constB=360}
   B=Multiply(A=Sign(None=@n97)) {constB=-1}
  )
  n99 = Distance(A=CameraPositionWS(), B=ActorPositionWS())
  n100 = ScalarParameter() {defaultValue=-5000, parameterName=ScaleMinDistanceOffset}
  n101 = Saturate(None=Divide(A=Add(A=@n99, B=@n100), B=ScalarParameter() {defaultValue=20000, parameterName=ScalDistance}))
  n102 = Add(A=Sine(None=LinearInterpolate(Alpha=@n101) {constA=-0.25, constB=0.25}))
  n103 = Multiply(
   A=Subtract(A=WorldPosition().XYZ, B=ActorPositionWS())
   B=LinearInterpolate(B=ScalarParameter() {defaultValue=1, parameterName=ScaleMax}, Alpha=Divide(A=@n102))
  )
  n104 = Add(A=WorldPosition().XYZ, B=@n103)
  n105 = LinearInterpolate(
   A=Add(A=Subtract(A=@n88, B=@n89.XYZ), B=Constant3Vector() {constant=<rgba 0 0 -5156156 0>})
   B=Add(A=RotateAboutAxis(NormalizedRotationAxis=@n92, RotationAngle=@n98, PivotPoint=@n94, Position=@n104), B=@n103)
   Alpha=StaticSwitchParameter(True=Constant() {r=1}, False=StaticSwitchParameter(True=@n91, False=@n90) {parameterName=isPhase2}) {parameterName=Debug}
  )
  n106 = ComponentMask(None=CollectionParameter() {collection=&MPC_CountdownTimerVisuals, parameterName=Timer_Color_Background}) {b=true, g=true, r=true}
  n107 = Multiply(
   A=ComponentMask(None=CollectionParameter() {collection=&MPC_CountdownTimerVisuals, parameterName=Timer_Color_Numbers}) {b=true, g=true, r=true}
   B=ScalarParameter() {defaultValue=1, group=Glitch, parameterName=Numbers.EmissiveMult}
  )
  n108 = ComponentMask(None=CollectionParameter() {collection=&MPC_CountdownTimerVisuals, parameterName=Timer_Color_Numbers}) {b=true, g=true, r=true}
  n109 = ComponentMask(None=CollectionParameter() {collection=&MPC_CountdownTimerVisuals, parameterName=Timer_Color_Image}) {b=true, g=true, r=true}
  n110 = TextureCoordinate()
  n111 = Constant2Vector() {g=0.5, r=1.1}
  n112 = OneMinus(None=@n111.R)
  n113 = Add(
   A=Multiply(A=@n110, B=@n111)
   B=MaterialFunctionCall(X=Divide(A=@n112), Y=Constant() {r=0.25}) {materialFunction=&MakeFloat2}.Result
  )
  n114 = NamedRerouteUsage()
  n115 = MaterialFunctionCall(
   UVs=LinearInterpolate(
    A=@n110
    B=Add(A=@n113, B=@n114)
    Alpha=CollectionParameter() {collection=&MPC_CountdownTimerVisuals, parameterName=Timer_NumbersGlitchAmount}
   )
   Radius=ScalarParameter() {defaultValue=0.5, group=Glitch, parameterName=Image.NumbersCutout.Color.Radius}
   Density=ScalarParameter() {defaultValue=100, group=Glitch, parameterName=Image.NumbersCutout.Color.Density}
  ) {materialFunction=&RadialGradientExponential}
  n116 = Multiply(
   A=LinearInterpolate(
    A=@n108
    B=LinearInterpolate(A=@n108, B=@n109, Alpha=@n115.RadialGradientExponential)
    Alpha=CollectionParameter() {collection=&MPC_CountdownTimerVisuals, parameterName=Timer_ImageCutoutNumbers}
   )
   B=ScalarParameter() {defaultValue=1, group=Glitch, parameterName=Numbers.EmissiveMult}
  )
  n117 = LinearInterpolate(
   A=@n116
   B=LinearInterpolate(A=@n107, B=@n116, Alpha=NamedRerouteUsage())
   Alpha=CollectionParameter() {collection=&MPC_CountdownTimerVisuals, parameterName=Timer_UseRandomFlicker}
  )
  n118 = CollectionParameter() {collection=&MPC_CountdownTimerVisuals, parameterName=Timer_ShowImage}
  n119 = CollectionParameter() {collection=&MPC_CountdownTimerVisuals, parameterName=Timer_Color_Text}
  n120 = LinearInterpolate(
   A=Multiply(A=LinearInterpolate(A=@n107, B=@n117, Alpha=@n118), B=NamedRerouteUsage())
   B=Multiply(A=ComponentMask(None=@n119) {b=true, g=true, r=true}, B=ScalarParameter() {defaultValue=1, group=Glitch, parameterName=Numbers.EmissiveMult})
   Alpha=NamedRerouteUsage()
  )
  n121 = CollectionParameter() {collection=&MPC_CountdownTimerVisuals, parameterName=Timer_Color_Image}
  n122 = NamedRerouteUsage()
  n123 = LinearInterpolate(A=Multiply(A=ComponentMask(None=@n121) {b=true, g=true, r=true}, B=@n122), B=@n122, Alpha=Saturate(None=ComponentMask(None=@n121) {a=true}))
  n124 = ScalarParameter() {defaultValue=1, group=Glitch, parameterName=Image.EmissiveMult}
  n125 = LinearInterpolate(
   B=Multiply(A=Multiply(A=@n123, B=@n124), B=NamedRerouteUsage())
   Alpha=CollectionParameter() {collection=&MPC_CountdownTimerVisuals, parameterName=Timer_ShowImage}
  )
  n126 = NamedRerouteUsage()
  n127 = NamedRerouteUsage()
  n128 = LinearInterpolate(
   A=@n127
   B=LinearInterpolate(B=@n127, Alpha=NamedRerouteUsage())
   Alpha=CollectionParameter() {collection=&MPC_CountdownTimerVisuals, parameterName=Timer_UseRandomFlicker}
  ) {constA=1}
  n129 = CollectionParameter() {collection=&MPC_CountdownTimerVisuals, parameterName=Timer_ShowImage}
  n130 = Saturate(None=LinearInterpolate(
    A=OneMinus(None=@n126)
    B=LinearInterpolate(B=@n128, Alpha=@n129)
    Alpha=CollectionParameter() {collection=&MPC_CountdownTimerVisuals, parameterName=Timer_ImageCutoutNumbers}
   ))
  n131 = NamedRerouteDeclaration(None=LinearInterpolate(A=@n106, B=LinearInterpolate(A=@n120, B=@n125, Alpha=@n130) {constAlpha=1}, Alpha=NamedRerouteUsage())) {name=FinalColor, nodeColor=<rgba 1 0.9844 0 1>}
  n132 = NamedRerouteUsage()
  n133 = NamedRerouteUsage()
  n134 = LinearInterpolate(
   B=LinearInterpolate(
    B=LinearInterpolate(Alpha=@n133) {constA=1, constB=0}
    Alpha=CollectionParameter() {collection=&MPC_CountdownTimerVisuals, parameterName=Timer_UseRandomFlicker}
   )
   Alpha=CollectionParameter() {collection=&MPC_CountdownTimerVisuals, parameterName=Timer_ShowImage}
  ) {constA=1, desc=Should Scanlines be visible}
  n135 = LinearInterpolate(A=@n131, B=Multiply(A=@n131, B=Multiply(A=@n132, B=@n134)), Alpha=ScalarParameter() {defaultValue=1, parameterName=Scanlines.OverlayMulti})
  n136 = Divide(A=@n135, B=EyeAdaptation().EyeAdaptation)
  n137 = Saturate(None=ScalarParameter() {defaultValue=1, parameterName=Emissive.UseEyeAdaptation})
  FunctionOutput(None=SetMaterialAttributes!(
    MaterialAttributes=SetMaterialAttributes!(MaterialAttributes, Opacity=@n87, WorldPositionOffset=@n105) {attributeSetTypes [#1, #2]}
    EmissiveColor=LinearInterpolate(A=@n135, B=@n136, Alpha=@n137)
   ) {attributeSetTypes [#3]}) {outputName=Material}
  FunctionOutput(None=LinearInterpolate(A=NamedRerouteUsage(), B=NamedRerouteUsage())) {outputName=Mask, sortPriority=1}
  FunctionOutput(None=NamedRerouteUsage()) {outputName=Color, sortPriority=2}
  FunctionOutput(None=NamedRerouteDeclaration(None=@n105) {name=FinalWPO, nodeColor=<rgba 1 0 0.7266 1>}) {outputName=Final WPO, sortPriority=4}
  n138 = Multiply(
   A=Time()
   B=CollectionParameter() {collection=&MPC_CountdownTimerVisuals, parameterName=Timer_FlickerSpeed}
  )
  n139 = Add(A=Multiply(A=@n138) {constB=1.66}) {constB=0.27}
  n140 = Multiply(A=SmoothStep(Value=Sine(None=@n138)) {constMax=0.2}, B=SmoothStep(Value=Sine(None=@n139)) {constMax=0.1})
  n141 = Multiply(A=@n138) {constB=2.11}
  n142 = SmoothStep(Value=Sine(None=Add(A=@n141) {constB=0.45})) {constMax=0.1}
  FunctionOutput(None=NamedRerouteDeclaration(None=Max(A=@n140, B=@n142)) {name=RandomFlicker, nodeColor=<rgba 0 0.9531 1 1>}) {outputName=Rand Flicker, sortPriority=50}
  LinearInterpolate(A=@n106, B=Multiply(A=@n106, B=NamedRerouteUsage()), Alpha=ScalarParameter() {parameterName=BackgroundFill})
  Multiply(
   A=@n132
   B=ComponentMask(None=CollectionParameter() {collection=&MPC_CountdownTimerVisuals, parameterName=Timer_Color_Numbers}) {b=true, g=true, r=true}
  )
  n143 = MaterialFunctionCall!(
   Width=ScalarParameter() {defaultValue=0.45, parameterName=Image.Fade.Width}
   Sharpness=ScalarParameter() {defaultValue=0.001, parameterName=Image.Fade.Sharpness}
   Offset
   Direction Switch
   Compare
   Input Coordinates
  ) {materialFunction=/Engine/Functions/Engine_MaterialFunctions03/Procedurals/GeneratedBand}
  n144 = TextureCoordinate()
  n145 = LinearInterpolate(
   A=@n144
   B=Add(A=@n144, B=NamedRerouteUsage())
   Alpha=CollectionParameter() {collection=&MPC_CountdownTimerVisuals, parameterName=Timer_ImageGlitchAmount}
  )
  n146 = AppendVector(
   A=ScalarParameter() {defaultValue=0.7524, group=Numbers, parameterName=u_PostImagePaddingScalea}
   B=ScalarParameter() {defaultValue=1.1963, group=Numbers, parameterName=v_PostImagePaddingScalea}
  )
  n147 = ScalarParameter() {group=Numbers, parameterName=u_PostImagePosOffset}
  n148 = ScalarParameter() {group=Numbers, parameterName=v_PostImagePosOffset}
  n149 = NamedRerouteDeclaration(None=Add(
    A=MaterialFunctionCall(UVs=@n145, Texture Scale=@n146) {materialFunction=&ScaleUVsByCenter}.UVs
    B=MaterialFunctionCall(X=@n147, Y=@n148) {materialFunction=&MakeFloat2}.Result
   )) {name=UV.BackgroundTexture, nodeColor=<rgba 0 0.6484 1 1>}
  n150 = Multiply(
   A=ScalarParameter() {defaultValue=0.03, group=Glitch, parameterName=Image.AbberationDistance}
   B=CollectionParameter() {collection=&MPC_CountdownTimerVisuals, parameterName=Timer_ChromAbberationPower}
  )
  n151 = TextureObjectParameter() {parameterName=Background, texture=&T_BlacksRGB}
  n152 = DDX(Value=@n149)
  n153 = DDY(Value=@n149)
  n154 = TextureSample!(
   UVs=Add(A=@n149, B=MaterialFunctionCall(X=@n150) {materialFunction=&MakeFloat2}.Result)
   Tex=@n151
   "DDX(UVs)"=@n152
   "DDY(UVs)"=@n153
   Apply View MipBias
  ) {mipValueMode=TMVM_Derivative, samplerSource=SSM_Clamp_WorldGroupSettings, texture=&T_BlacksRGB}
  n155 = TextureSample(UVs=@n149, Tex=@n151, "DDX(UVs)"=@n152, "DDY(UVs)"=@n153) {mipValueMode=TMVM_Derivative, samplerSource=SSM_Clamp_WorldGroupSettings, texture=&T_BlacksRGB}
  n156 = Multiply(A=@n150) {constB=-1}
  n157 = TextureSample(
   UVs=Add(A=@n149, B=MaterialFunctionCall(X=@n156) {materialFunction=&MakeFloat2}.Result)
   Tex=@n151
   "DDX(UVs)"=@n152
   "DDY(UVs)"=@n153
  ) {mipValueMode=TMVM_Derivative, samplerSource=SSM_Clamp_WorldGroupSettings, texture=&T_BlacksRGB}
  n158 = MaterialFunctionCall(
   X=StaticComponentMaskParameter(None=@n154.RGBA) {defaultA=true, group=Glitch, parameterName=Image.MaskTextureChannel}
   Y=StaticComponentMaskParameter(None=@n155.RGBA) {defaultA=true, group=Glitch, parameterName=Image.MaskTextureChannel}
   Z=StaticComponentMaskParameter(None=@n157.RGBA) {defaultA=true, group=Glitch, parameterName=Image.MaskTextureChannel}
  ) {materialFunction=&MakeFloat3}
  n159 = LinearInterpolate(A=@n158.Result, B=Multiply(A=@n158.Result, B=@n143.Result), Alpha=ScalarParameter() {parameterName=Image.FadeEdges.Alpha})
  NamedRerouteDeclaration(None=Desaturation(None=Saturate(None=@n159))) {name=Image.Mask, nodeColor=<rgba 1 0 0.9141 1>}
  NamedRerouteDeclaration(None=Saturate(None=@n51)) {name=Numbers.Text.Mask, nodeColor=<rgba 0.1953 0 1 1>}
  n160 = MaterialFunctionCall(X=@n154.R, Y=@n155.G, Z=@n157.B) {materialFunction=&MakeFloat3}
  NamedRerouteDeclaration(None=LinearInterpolate(A=@n160.Result, B=Multiply(A=@n160.Result, B=@n143.Result), Alpha=ScalarParameter() {parameterName=Image.FadeEdges.Alpha})) {name=Image.Color, nodeColor=<rgba 1 0 0.375 1>}
  n161 = NamedRerouteUsage()
  n162 = NamedRerouteDeclaration(None=@n54) {name=OutForHololines, nodeColor=<rgba 0.5938 0 1 1>}
  n163 = LinearInterpolate(A=@n161, B=@n162, Alpha=NamedRerouteUsage())
  n164 = CollectionParameter() {collection=&MPC_CountdownTimerVisuals, parameterName=Timer_UseRandomFlicker}
  NamedRerouteDeclaration(None=LinearInterpolate(
    A=@n161
    B=LinearInterpolate(A=@n162, B=@n163, Alpha=@n164)
    Alpha=CollectionParameter() {collection=&MPC_CountdownTimerVisuals, parameterName=Timer_ShowImage}
   )) {name=FinalMask, nodeColor=<rgba 0 1 0.3203 1>}
  n165 = TextureCoordinate()
  n166 = LinearInterpolate(A=@n165, B=Add(A=@n165, B=NamedRerouteUsage()))
  n167 = Constant() {r=0.5}
  n168 = MaterialFunctionCall(X=Constant() {r=1}, Y=@n167) {materialFunction=&MakeFloat2}
  n169 = Constant() {r=0.5}
  n170 = Divide(A=@n167)
  NamedRerouteDeclaration(None=MaterialFunctionCall(
    UVs=Multiply(A=@n166, B=@n168.Result)
    CenterPosition=MaterialFunctionCall(X=@n169, Y=@n170) {materialFunction=&MakeFloat2}.Result
    Radius=ScalarParameter() {defaultValue=0.5, group=Glitch, parameterName=Image.NumbersCutout.Radius}
    Density=ScalarParameter() {defaultValue=100, group=Glitch, parameterName=Image.NumbersCutout.Density}
   ) {materialFunction=&RadialGradientExponential}.RadialGradientExponential) {name=Mask.NumberCutout, nodeColor=<rgba 0 1 0.0625 1>}
  Round(None=@n130)
 ]}
