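# M_Countdown_Master_F_LimeRock is an unlit translucent master material for the LimeRock countdown timer, apparently giving it a hologram look.
# It uses material attributes, and its only node is a call to the MF_LimeRockCountdownHologram material function, which holds the shading logic.
# Responsive AA is enabled and the material is flagged for use with instanced static meshes.
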Material /LimeRock_CountdownTimer/Gameplay/CountdownTimer/M_Countdown_Master_F_LimeRock
graph {
 properties {
  bEnableResponsiveAA=true
  bUseMaterialAttributes=true
  bUsedWithInstancedStaticMeshes=true
  blendMode=BLEND_Translucent
  pixelDepthOffsetMode=PDOM_Legacy
  shadingModel=MSM_Unlit
  usedShadingModels=Unlit
 }
 nodes [MaterialFunctionCall() {materialFunction=&/MF_LimeRockCountdownHologram}]
}
